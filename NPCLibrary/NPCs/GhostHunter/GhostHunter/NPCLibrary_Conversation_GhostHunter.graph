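# Conversation database for the GhostHunter NPC, Torin, a monster hunter who greets the player and leads into shared NPC dialogue.
# A single entry point feeds one speech node with a curt default line and contextual greetings gated by controller requirements for time since last match, Sideways visit, first conversation and held Bars.
# Each greeting's priority comes from a curve table row.
# The speech node then links out to the shared NPC quests and basic services conversations by entry tag.
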ConversationDatabase /NPCLibrary/NPCs/GhostHunter/GhostHunter/NPCLibrary_Conversation_GhostHunter
entry_points [{tag=Conversation.Entry.GhostHunter, destinations [#1]}]
nodes [
 ConversationEntryPointNode_7 {
  id=#1
  class=ConversationEntryPointNode
  outputs [#2]
  properties {entryTag {tagName=Conversation.Entry.GhostHunter}}
 }
 Speech_C_4 {
  id=#2
  class=Speech_C
  outputs [#3, #4]
  properties {
   nodeName=""
   generalConfig {
    defaultMessage=What do you want? I'm kinda busy.
    contextualMessages [
     {
      message=You're joining the hunt? Good.
      contextRequirements [{requirement {class=FortControllerRequirement_CheckTimeSinceLastPlayedMatch}}]
      priority {curve {rowName=Tandem.Speech.Priority.WhileSincePlayed}}
     }
     {
      message=Anything new to report from the Sideways?
      contextRequirements [{requirement {class=FortControllerRequirement_HasOwnedTag, properties {ownedTag {tagName=Granted.Athena.Sideways.Encounter.Visited}}}}]
      priority {curve {rowName=Tandem.Speech.Priority.VisitedSpecificPOI}}
     }
     {
      message=Name's Torin. I hunt monsters. Let's keep this strictly business.
      contextRequirements [{requirement {class=FortControllerRequirement_IsFirstEverConversationWithNPC}}]
      priority {curve {rowName=Tandem.Speech.Priority.FirstEverInteraction}}
     }
     {
      message="If you wanna use those Bars for good, I have some research projects..."
      contextRequirements [{requirement {class=FortControllerRequirement_HasBars}}]
      priority {curve {rowName=Tandem.Speech.Priority.HighTotalBars}}
     }
    ]
   }
  }
 }
 ConversationLinkNode_0 {id=#4, class=ConversationLinkNode, properties {remoteEntryTag {tagName=Conversation.Entry.BasicServices}}}
 ConversationLinkNode_1 {id=#3, class=ConversationLinkNode, properties {remoteEntryTag {tagName=Conversation.Entry.NPCLibrary.Quests}}}
]
